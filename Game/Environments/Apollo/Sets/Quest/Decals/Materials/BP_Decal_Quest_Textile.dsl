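; Quest decal Blueprint for the Apollo Textile decal that tags players standing in front of it for quest tracking.
; When a player pawn enters its Box component, it applies GE_Decal_Quest_Textile through that pawn's ability system component.
; On leaving the box, it removes active effects that grant the Athena.Quests.Decal.Textile tag.
; BeginPlay, Tick, actor overlap and the construction script are empty.
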(event EventBeginPlay)

(event Collision|EventActorBeginOverlap
  (params (OtherActor :type Actor%20Object%20Reference)))

(event EventTick
  (params (DeltaSeconds :type Float%20%28single-precision%29)))

(event OnComponentBeginOverlap(Box)
  (params (OverlappedComponent :type Primitive%20Component%20Object%20Reference) (OtherActor :type Actor%20Object%20Reference) (OtherComp :type Primitive%20Component%20Object%20Reference) (OtherBodyIndex :type Integer) (bFromSweep :type Boolean) (SweepResult :type Hit%20Result%20Structure%20%28by%20ref%29))
  ;[1 1 1] Add GE to players infront of Decal for quest tracking
  (Utilities|IsValid OtherActor
    (:Is%20Valid
      (bind _asfort_player_pawn (Utilities|Casting|CastToFortPlayerPawn OtherActor))
      (GameplayEffects|ApplyGameplayEffectToTarget :self (Ability|GetAbilitySystemComponent _asfort_player_pawn) :GameplayEffectClass "/Game/Environments/Apollo/Sets/Quest/Decals/Materials/GE_Decal_Quest_Textile.GE_Decal_Quest_Textile_C" :Target (Ability|GetAbilitySystemComponent _asfort_player_pawn) :Level 0.0))))

(event OnComponentEndOverlap(Box)
  (params (OverlappedComponent :type Primitive%20Component%20Object%20Reference) (OtherActor :type Actor%20Object%20Reference) (OtherComp :type Primitive%20Component%20Object%20Reference) (OtherBodyIndex :type Integer))
  ;[1 1 1] Remove GE from players when no longer in front of Decal
  (Utilities|IsValid OtherActor
    (:Is%20Valid
      (bind _asfort_player_pawn (Utilities|Casting|CastToFortPlayerPawn OtherActor))
      (GameplayEffects|RemoveActiveEffectswithGrantedTags (Ability|GetAbilitySystemComponent _asfort_player_pawn) "(GameplayTags=((TagName=\"Athena.Quests.Decal.Textile\")))"))))

(fn ConstructionScript)
